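# Niagara swing trail effect for the SteedShimmer pickaxe, built from two short-lived animation-trail ribbon emitters.
# Trail draws the main ribbon, while TipTrail drives three ribbon renderers and narrows over particle life.
# User.IsVariant switches colors on both emitters and width on TipTrail.
# Both fade quickly with time-of-day tinting, and TipTrail also writes bright WaveColor and Trail_Tip colors whose alpha combines particle curves with system-age curves.
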NiagaraSystem /BRCosmetics/Weapons/FORT_Melee/Pickaxe_SteedShimmer/FX/NS_Pickaxe_SteedShimmer_Swing
dependencies {
 usedDataInterfaces [&NiagaraDataInterfaceCurve, &NiagaraDataInterfaceColorCurve, &NiagaraDataInterfaceSocketReader]
 usedDynamicInputs [&SetBoolByFloatComparison]
 usedModules [
  &self:SystemSpawnScript.SystemScriptSource.SystemScriptGraph.NiagaraNodeAssignment_0.SetVariables_#1
  &self:SystemSpawnScript.SystemScriptSource.SystemScriptGraph.NiagaraNodeAssignment_1.SetVariables_#2
  &SystemState
 ]
}
emitters {
 TipTrail {
  emitterUpdateScript [
   &EmitterState(Life Cycle Mode=::NewEnumerator1, Loop Behavior=::NewEnumerator1, Loop Duration=0.5, UseLoopDelay=false)
   &FN_SpawnAnimTrail(AnimTrailWidthMode=::FromSecond)
  ]
  particleSpawnScript [
   &FN_InitializeAnimTrailBoneIndecies()
   &FN_InitializeAnimTrail(
    AnimTrailWidth=&MergeHelper_FloatSwitch(0=0.1, 1=0.15, 2=?, 3=?, Index=&MakeIntFromFloat(Float=$User.IsVariant), Num=::NewEnumerator0)
    Color=&MergeHelper_ColorSwitch(
     0=<rgba 0.227 0.4125 0.7011 1>
     1=<rgba 20 1 18.4167 0.35>
     2=?
     3=?
     Index=&MakeIntFromFloat(Float=$User.IsVariant)
     Num=::NewEnumerator0
    )
    Lifetime=0.45
   )
  ]
  particleUpdateScript [
   &FN_UpdateAnimTrail()
   &SolveForcesAndVelocity(Clamp Velocity=false, Limit Acceleration=false)
   &ScaleRibbonWidth(Ribbon Width Scale=&FloatFromCurve(FloatCurve=DataInterface() {propertyValues {Curve {keys [
         {arriveTangent=-1.2579, interpMode=RCIM_Cubic, time=0, value=1}
         {arriveTangent=-0.9664, interpMode=RCIM_Cubic, leaveTangent=-0.9664, tangentMode=RCTM_User, time=1, value=0}
        ]}}}))
   &ScaleColor(
    Scale Alpha=&Multiply_Float(
     A=&FloatFromCurve(FloatCurve=DataInterface() {propertyValues {Curve {keys [
          {interpMode=RCIM_Cubic, leaveTangent=-2.3646, tangentMode=RCTM_Break, time=0, value=1}
          {arriveTangent=-10.2962, interpMode=RCIM_Cubic, leaveTangent=-10.2962, tangentMode=RCTM_User, time=0.1129, value=0.5809}
          <at 0.25 0 RCIM_Cubic>
         ]}}})
     B=&FNTimeOfDayScalar(Night Scale=0.5, Morning Scale=0.6, Evening Scale=0.5)
    )
    Scale RGB=&MakeVectorFromLinearColorRGB(Color=&FNTimeOfDayColor(
      Night Color=<rgba 0.4463 0.62 1 1>
      Morning Color=<rgba 1 0.6899 0.5573 1>
      Day Color=<rgba 1 0.7878 0.6667 1>
      Evening Color=<rgba 1 0.6291 0.4464 1>
     ))
    ScaleA=true
    ScaleRGB=true
   ) {module=ScaleColor002}
   &ScaleColor(
    Linear Color Curve=DataInterface() {propertyValues {
      AlphaCurve {keys [<at 0 1>,<at 0.8009 0>]}
      BlueCurve {keys [<at 0 1>,<at 0.2665 0.0802>]}
      GreenCurve {keys [<at 0 1>,<at 0.2665 0.0602>]}
      RedCurve {keys [<at 0 1>,<at 0.2665 0.0678>]}
     }}
    Scale Mode=::NewEnumerator2
   )
   /SetVariables(Particles.WaveColor:LinearColor=&MultiplyLinearColor(
     Linear Color=<rgba 20 10 20 1>
     Scale Factor=&MakeVector4FromVectorAndFloat(
      "Vector (RGB)"=<1 1 1>
      "Float (Alpha)"=&Multiply_Float(
       A=&FloatFromCurve(FloatCurve=DataInterface() {propertyValues {Curve {keys [
            {arriveTangent=15.6615, interpMode=RCIM_Cubic, leaveTangent=15.6615, tangentMode=RCTM_User, time=0, value=1.5}
            {arriveTangent=-6.5588, interpMode=RCIM_Cubic, leaveTangent=-6.5588, tangentMode=RCTM_User, time=0.065, value=0.5187}
            <at 0.3 0 RCIM_Cubic>
           ]}}})
       B=&FloatFromCurve(
        FloatCurve=DataInterface() {propertyValues {Curve {keys [
            {arriveTangent=-1.2576, interpMode=RCIM_Cubic, leaveTangent=-1.2576, tangentMode=RCTM_User, time=0.1, value=1}
            {arriveTangent=-0.4242, interpMode=RCIM_Cubic, leaveTangent=-0.4242, tangentMode=RCTM_User, time=0.1428, value=0.2211}
            <at 0.35 0 RCIM_Cubic>
           ]}}}
        CurveIndex=$Engine.System.Age
        Scale Curve=0.2
       )
      )
     )
    )) {isSetParameters=true}
   /SetVariables(Particles.Position_Tip:NiagaraPosition=&AddVectorToPosition(Position=$Particles.Position, Vector=&TransformVector(Vector=<0 0 0>, Destination Space=::Local))) {isSetParameters=true}
   /SetVariables(Particles.Trail_Tip:LinearColor=&MultiplyLinearColor(
     Linear Color=<rgba 35 17.5 32.0833 1>
     Scale Factor=&MakeVector4FromVectorAndFloat(
      "Vector (RGB)"=<1 1 1>
      "Float (Alpha)"=&Multiply_Float(
       A=&FloatFromCurve(FloatCurve=DataInterface() {propertyValues {Curve {keys [
            {interpMode=RCIM_Cubic, leaveTangent=-2.3646, tangentMode=RCTM_Break, time=0, value=1}
            {arriveTangent=-10.2962, interpMode=RCIM_Cubic, leaveTangent=-10.2962, tangentMode=RCTM_User, time=0.1129, value=0.5809}
            <at 0.25 0 RCIM_Cubic>
           ]}}})
       B=&FloatFromCurve(
        FloatCurve=DataInterface() {propertyValues {Curve {keys [
            {interpMode=RCIM_Cubic, leaveTangent=-2.3646, tangentMode=RCTM_Break, time=0, value=1}
            {arriveTangent=-10.2962, interpMode=RCIM_Cubic, leaveTangent=-10.2962, tangentMode=RCTM_User, time=0.1129, value=0.5809}
            <at 0.25 0 RCIM_Cubic>
           ]}}}
        CurveIndex=$Engine.System.Age
        Scale Curve=0.2
       )
      )
     )
    )) {isSetParameters=true}
  ]
  rendererClasses [&NiagaraRibbonRendererProperties]
  renderers [
   {rendererClass=&NiagaraRibbonRendererProperties}
   {rendererClass=&NiagaraRibbonRendererProperties, rendererIndex=1}
   {rendererClass=&NiagaraRibbonRendererProperties, rendererIndex=2}
  ]
 }
 Trail {
  emitterUpdateScript [
   &EmitterState(Life Cycle Mode=::NewEnumerator1, Loop Behavior=::NewEnumerator1, Loop Duration=0.5, UseLoopDelay=false)
   &FN_SpawnAnimTrail()
  ]
  particleSpawnScript [
   &FN_InitializeAnimTrailBoneIndecies()
   &FN_InitializeAnimTrail(
    Color=&MergeHelper_ColorSwitch(
     0=<rgba 0.0331 0.0331 0.0762 1>
     1=<rgba 0.0331 0.0331 0.0762 2>
     2=?
     3=?
     Index=&MakeIntFromFloat(Float=$User.IsVariant)
     Num=::NewEnumerator0
    )
    Lifetime=0.45
   )
  ]
  particleUpdateScript [
   &FN_UpdateAnimTrail()
   &SolveForcesAndVelocity(Clamp Velocity=false, Limit Acceleration=false)
   &ScaleColor(
    Scale Alpha=&Multiply_Float(
     A=&FloatFromCurve(FloatCurve=DataInterface() {propertyValues {Curve {keys [
          {interpMode=RCIM_Cubic, leaveTangent=7.2426, tangentMode=RCTM_Break, time=0, value=2}
          {arriveTangent=-5.9968, interpMode=RCIM_Cubic, leaveTangent=-5.9968, tangentMode=RCTM_User, time=0.1156, value=0.3206}
          {interpMode=RCIM_Cubic, leaveTangent=-2.9059, time=0.25, value=0}
         ]}}})
     B=&FNTimeOfDayScalar(Night Scale=0.5, Morning Scale=0.6, Evening Scale=0.5)
    )
    Scale RGB=&MakeVectorFromLinearColorRGB(Color=&FNTimeOfDayColor(
      Night Color=<rgba 0.4463 0.62 1 1>
      Morning Color=<rgba 1 0.6899 0.5573 1>
      Day Color=<rgba 1 0.7878 0.6667 1>
      Evening Color=<rgba 1 0.6291 0.4464 1>
     ))
    ScaleA=true
    ScaleRGB=true
   ) {module=ScaleColor002}
  ]
  rendererClasses [&NiagaraRibbonRendererProperties]
  renderers [{rendererClass=&NiagaraRibbonRendererProperties}]
 }
}
summary {
 systemName=NS_Pickaxe_SteedShimmer_Swing
 emitters [
  {emitterName=Trail, rendererClasses [&NiagaraRibbonRendererProperties]}
  {emitterName=TipTrail, rendererClasses [&NiagaraRibbonRendererProperties]}
 ]
 userVariables [
  User.IsVariant {type=NiagaraFloat, defaultValue=0, description=""}
  User.RibbonIDOffset {type=NiagaraInt32}
 ]
}
